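# Material function for the MagmaBreak character skin that adds animated lava emissive and optional vertex displacement to incoming material attributes, and outputs a magma mask.
# Pre-skinned local position drives polar UVs for distorted panning noise, sampled with explicit derivatives, while a hair switch uses rotated mesh UVs.
# FX and emissive textures feed color-ramped lava, rims and cavity glow.
# An optional backpack reactivity switch adds a timestamp-driven fade in and out with sphere-based masks.
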MaterialFunction /BRCosmetics/Characters/Player/Male/Medium/Bodies/M_MED_MagmaBreak/Materials/MF_MagmaBreak
graph {nodes [
  FunctionOutput(None=SetMaterialAttributes!(MaterialAttributes=NamedRerouteUsage(), EmissiveColor=NamedRerouteUsage(), WorldPositionOffset=NamedRerouteUsage()) {attributeSetTypes [#1, #2]})
  FunctionOutput(None=Saturate(None=NamedRerouteUsage())) {outputName=MagmaMask, sortPriority=1}
  n1 = NamedRerouteUsage()
  n2 = NamedRerouteUsage()
  n3 = NamedRerouteDeclaration(None=ChannelMaskParameter(None=NamedRerouteUsage()) {group=MagmaBreak, maskChannel=Green, parameterName=MagmaFXMask, defaultValue=<rgba 0 1 0 0>}.R) {name=MagmaMask, nodeColor=<rgba 1 0.1641 0 1>}
  n4 = Multiply(
   A=StaticSwitchParameter(True=LinearInterpolate(A=@n1, Alpha=@n2), False=Constant() {r=1}) {group=MagmaBreak, parameterName=UseBackpackReactivity}
   B=Add(A=@n3, B=NamedRerouteUsage())
  )
  NamedRerouteDeclaration(None=@n4) {name=OutMask, nodeColor=<rgba 0 1 0.4844 1>}
  NamedRerouteDeclaration(None=FunctionInput() {inputType=FunctionInput_MaterialAttributes}) {name=InMaterialAttributes, nodeColor=<rgba 1 0 0.3047 1>}
  n5 = Add(
   A=LocalPosition() {localOrigin=InstancePreSkinning}.XYZ
   B=VectorParameter() {group=MagmaBreak, parameterName=PositionOffset, defaultValue=<rgba 0 0 -1.5 1>}.RGB
  )
  n6 = VertexInterpolator(VS=@n5)
  n7 = Divide(A=@n5, B=ScalarParameter() {defaultValue=1, group=MagmaBreak, parameterName=WPONoiseScale}) {desc=AdjustHairWPO}
  n8 = Divide(
   A=NamedRerouteDeclaration(None=ShaderStageSwitch(PixelShader=@n6.PS, VertexShader=@n7)) {name=PreSkinnedPosition, nodeColor=<rgba 0 0.3438 1 1>}
   B=ScalarParameter() {defaultValue=7.5, group=MagmaBreak, parameterName=MagmaNoiseTiling}
  ) {constB=40}
  n9 = Subtract(
   A=LinearInterpolate(
    A=ComponentMask(None=@n8) {b=true, r=true}
    B=ComponentMask(None=@n8) {b=true, g=true}
    Alpha=ScalarParameter() {group=MagmaBreak, parameterName=UseYAxis}
   )
   B=Constant2Vector() {desc=Center, g=-0.5, r=-0.5}
  ) {desc=Delta}
  n10 = SphereMask(A=@n9, B=Constant() {r=0.5}) {attenuationRadius=1.5, hardnessPercent=0}
  NamedRerouteDeclaration(None=OneMinus(None=Multiply(A=@n10, B=@n10))) {name=Core, nodeColor=<rgba 0.7109 0 1 1>}
  n11 = MaterialFunctionCall(Color=NamedRerouteUsage()) {materialFunction=&SplitComponents}
  n12 = LinearInterpolate(A=@n11.G, B=ComponentMask(None=NamedRerouteUsage()) {r=true}, Alpha=ScalarParameter() {group=MagmaBreak, parameterName=InnerRimUseAO})
  n13 = Multiply(
   A=Power(Base=OneMinus(None=@n12), Exp=ScalarParameter() {defaultValue=7, group=MagmaBreak, parameterName=InnerRimExponent}) {constExponent=4}
   B=ScalarParameter() {defaultValue=1, group=MagmaBreak, parameterName=InnerRimIntensity}
  )
  n14 = Add(B=NamedRerouteUsage()) {constA=1}
  NamedRerouteDeclaration(None=Saturate(None=Multiply(A=@n13, B=@n14))) {name=InnerRim, nodeColor=<rgba 1 0.7969 0 1>}
  n15 = TextureCoordinate()
  n16 = Constant()
  NamedRerouteDeclaration(None=ShaderStageSwitch(PixelShader=DDX(Value=@n15), VertexShader=@n16)) {name=DDX, nodeColor=<rgba 0.7109 1 0 1>}
  NamedRerouteDeclaration(None=ShaderStageSwitch(PixelShader=DDY(Value=@n15), VertexShader=@n16)) {name=DDY, nodeColor=<rgba 0.9688 0 1 1>}
  n17 = TextureSampleParameter2D() {
   group=MagmaBreak
   parameterName=MagmaEmissiveTexture
   texture=/BRCosmetics/Characters/Player/Male/Medium/Bodies/M_MED_MagmaBreak/Textures/T_MagmaBreak_Body_E
  }
  NamedRerouteDeclaration(None=@n17.RGB) {name=InEmissive, nodeColor=<rgba 0 1 0.6484 1>}
  n18 = ComponentMask(None=NamedRerouteUsage()) {r=true}
  n19 = ScalarParameter() {group=MagmaBreak, parameterName=InnerGlowUseAO}
  n20 = NamedRerouteDeclaration(None=Multiply(A=LinearInterpolate(A=@n11.B, B=@n18, Alpha=@n19), B=NamedRerouteUsage())) {name=CavityCenter, nodeColor=<rgba 0.6875 1 0 1>}
  n21 = ScalarParameter() {defaultValue=2, group=MagmaBreak, parameterName=InnerGlowExponent}
  n22 = ScalarParameter() {defaultValue=200, group=MagmaBreak, parameterName=InnerGlowIntensity}
  n23 = Max(A=NamedRerouteUsage()) {constB=0.5}
  NamedRerouteDeclaration(None=Multiply(A=Power(Base=@n20, Exp=@n21), B=Multiply(A=@n22, B=@n23)) {constB=5}) {name=InnerGlow, nodeColor=<rgba 0.6172 0 1 1>}
  n24 = Multiply(
   A=ChannelMaskParameter(None=NamedRerouteUsage()) {group=MagmaBreak, maskChannel=Blue, parameterName=ExtremityFixMask, defaultValue=<rgba 0 0 1 0>}.R
   B=ScalarParameter() {defaultValue=1, group=MagmaBreak, parameterName=ExtremityFixIntensity}
  )
  NamedRerouteDeclaration(None=StaticSwitchParameter(True=Saturate(None=@n24), False=Constant()) {defaultValue=true, group=MagmaBreak, parameterName=UseExtremitiesFix}) {name=ExtremitiesFix, nodeColor=<rgba 0 1 0.0859 1>}
  n25 = Constant() {r=0.25}
  n26 = Constant() {r=1}
  n27 = VectorParameter() {group=MagmaBreak, parameterName=UVScale, defaultValue=<rgba 1 1 0 1>}
  NamedRerouteDeclaration(None=Multiply(
    A=MaterialFunctionCall!(UV Coord=@n9, LenghtScale=@n25, RadialScale=@n26) {materialFunction=/Game/Global/MaterialFunctions/TechArt/MF_PolarCoordinates}.Result
    B=ComponentMask(None=@n27.RGB) {g=true, r=true}
   )) {name=Radial UVs, nodeColor=<rgba 1 0 0.7734 1>}
  n28 = Multiply(A=Time(), B=Constant() {r=0.75})
  n29 = Multiply(A=ConstantBiasScale(None=Sine(None=@n28))) {constB=0.5}
  n30 = Multiply(A=ConstantBiasScale(None=Sine(None=@n28) {period=2})) {constB=0.25}
  NamedRerouteDeclaration(None=Saturate(None=Add(A=@n29, B=@n30))) {name=Frequency, nodeColor=<rgba 1 0.9609 0 1>}
  n31 = Multiply(A=NamedRerouteUsage(), B=ScalarParameter() {defaultValue=0.6, group=MagmaBreak, parameterName=WPOIntensity})
  n32 = Transform(None=NamedRerouteUsage())
  n33 = Multiply(
   A=LinearInterpolate(A=@n31, B=Multiply(A=@n32, B=@n31), Alpha=ScalarParameter() {defaultValue=1, group=MagmaBreak, parameterName=UseWPONormal})
   B=OneMinus(None=NamedRerouteUsage())
  )
  n34 = ComponentMask(None=NamedRerouteUsage()) {b=true}
  n35 = StaticSwitchParameter(True=OneMinus(None=Multiply(A=@n34, B=@n34)), False=Constant() {r=1}) {group=MagmaBreak, parameterName=UseHair}
  NamedRerouteDeclaration(None=StaticSwitchParameter(True=LinearInterpolate(B=@n33, Alpha=@n35), False=Constant()) {group=MagmaBreak, parameterName=UseWPO}) {name=OutWPO, nodeColor=<rgba 1 0.0469 0 1>}
  n36 = NamedRerouteUsage()
  n37 = TextureCoordinate()
  n38 = ScalarParameter() {defaultValue=7.5, group=MagmaBreak, parameterName=MagmaNoiseTiling}
  n39 = VectorParameter() {group=MagmaBreak, parameterName=HairUVScale, defaultValue=<rgba 1 1 0 1>}
  n40 = MaterialFunctionCall(
   UVs=Multiply(A=Multiply(A=@n37, B=@n38) {constB=2}, B=ComponentMask(None=@n39.RGB) {g=true, r=true})
   "Rotation Angle (0-1)"=Constant() {r=0.25}
  ) {materialFunction=&CustomRotator}
  n41 = TextureSample(UVs=@n40.Rotated Values) {samplerType=SAMPLERTYPE_LinearColor, texture=&/T_MagmaBreak_Mask}
  n42 = ScalarParameter() {defaultValue=0.05, group=MagmaBreak, parameterName=NoiseDistortionAmount}
  n43 = Subtract(
   A=Multiply(A=Subtract(A=@n41.R) {constB=0.5}, B=@n42) {constB=0.05}
   B=Multiply(A=NamedRerouteUsage(), B=ScalarParameter() {defaultValue=1, group=MagmaBreak, parameterName=EmissiveDisplacementContribution})
  )
  n44 = VectorParameter() {group=MagmaBreak, parameterName=MagmaHairFlowSpeed, defaultValue=<rgba -0.05 0 0 1>}
  n45 = TextureSample(UVs=Panner(Coordinate=Add(A=@n40.Rotated Values, B=@n43), Speed=ComponentMask(None=@n44.RGB) {g=true, r=true}) {speedX=0.1}) {samplerType=SAMPLERTYPE_LinearColor, texture=&T_MagmaBreak_Mask}
  n46 = Saturate(None=Power(Base=SmoothStep(Value=@n45.G), Exp=ScalarParameter() {defaultValue=4, group=MagmaBreak, parameterName=MagmaNoiseExp}) {constExponent=4})
  n47 = ScalarParameter() {group=MagmaBreak, parameterName=MagmaNoiseFade}
  n48 = Multiply(A=Saturate(None=Add(A=@n46, B=@n47)), B=ScalarParameter() {defaultValue=4, group=MagmaBreak, parameterName=MagmaNoiseInt})
  n49 = ScalarParameter() {defaultValue=-0.5, group=MagmaBreak, parameterName=HairBaseMin}
  n50 = ScalarParameter() {defaultValue=1.1, group=MagmaBreak, parameterName=HairBaseMax}
  n51 = Desaturation(None=NamedRerouteUsage(), Fraction=Constant() {r=1})
  n52 = Multiply(
   A=Power(
    Base=LinearInterpolate(A=@n49, B=@n50, Alpha=@n51) {constA=-0.5, constB=1.1}
    Exp=ScalarParameter() {defaultValue=1, group=MagmaBreak, parameterName=HairBaseExp}
   )
   B=ScalarParameter() {defaultValue=1, group=MagmaBreak, parameterName=HairBaseIntensity}
  ) {constB=2}
  n53 = Panner(
   Coordinate=NamedRerouteUsage()
   Speed=ComponentMask(None=VectorParameter() {group=MagmaBreak, parameterName=MagmaFlowSpeed, defaultValue=<rgba -0.075 0.05 0 1>}.RGB) {g=true, r=true}
  ) {speedX=-0.075}
  n54 = TextureSample!(UVs=NamedRerouteUsage(), Tex, "DDX(UVs)"=NamedRerouteUsage(), "DDY(UVs)"=NamedRerouteUsage(), Apply View MipBias) {
   mipValueMode=TMVM_Derivative
   samplerType=SAMPLERTYPE_LinearColor
   texture=&T_MagmaBreak_Mask
  }
  n55 = LinearInterpolate(
   A=Multiply(A=Subtract(A=@n54.G) {constB=0.5}, B=@n42) {constB=0.05}
   Alpha=Multiply(A=NamedRerouteUsage(), B=ScalarParameter() {defaultValue=1, group=MagmaBreak, parameterName=EmissiveDisplacementContribution})
  ) {constB=-1}
  n56 = TextureSample(
   UVs=NamedRerouteDeclaration(None=Add(A=@n53, B=@n55)) {name=RadialOffsettedUVs, nodeColor=<rgba 0.4531 1 0 1>}
   "DDX(UVs)"=NamedRerouteUsage()
   "DDY(UVs)"=NamedRerouteUsage()
  ) {
   mipValueMode=TMVM_Derivative
   samplerType=SAMPLERTYPE_LinearColor
   texture=&T_MagmaBreak_Mask
  }
  n57 = LinearInterpolate(B=@n56.G, Alpha=NamedRerouteUsage()) {constA=1}
  n58 = ScalarParameter() {defaultValue=4, group=MagmaBreak, parameterName=MagmaNoiseExp}
  n59 = Add(
   A=NamedRerouteDeclaration(None=Power(Base=@n57, Exp=@n58) {constExponent=4}) {name=PanningNoise, nodeColor=<rgba 0 1 0.3672 1>}
   B=NamedRerouteUsage()
  )
  n60 = Saturate(None=Multiply(A=NamedRerouteUsage(), B=ScalarParameter() {defaultValue=1, group=MagmaBreak, parameterName=InnerRimIntensity}))
  n61 = StaticSwitchParameter(
   True=LinearInterpolate(B=Multiply(A=@n48, B=@n52), Alpha=@n52)
   False=Multiply(A=LinearInterpolate(A=@n59, Alpha=@n60) {constB=0}, B=NamedRerouteUsage())
  ) {group=MagmaBreak, parameterName=UseHair}
  n62 = Multiply(A=LinearInterpolate(A=Multiply(A=@n36) {constB=0.4}, Alpha=NamedRerouteUsage()), B=@n61) {desc=For Backpack}
  NamedRerouteDeclaration(None=LinearInterpolate(
    A=VectorParameter() {group=MagmaBreak, parameterName=Color1, defaultValue=<rgba 0.5521 0.0167 0 1>}.RGB
    B=VectorParameter() {group=MagmaBreak, parameterName=Color2, defaultValue=<rgba 1 0.5944 0 1>}.RGB
    Alpha=StaticSwitchParameter(True=@n62, False=@n61) {group=MagmaBreak, parameterName=UseBackpackReactivity}
   )) {name=OutPanningLava, nodeColor=<rgba 0 0.2734 1 1>}
  n63 = SphereMask(A=Add(A=NamedRerouteUsage(), B=Constant3Vector() {constant=<rgba 0 15 -0.25 0>}), B=Constant()) {attenuationRadius=23, hardnessPercent=0}
  n64 = Saturate(None=InverseLinearInterpolate(Value=@n63) {constA=0.75})
  n65 = Multiply(A=@n64, B=@n64)
  NamedRerouteDeclaration(None=Saturate(None=Multiply(A=@n65) {constB=10})) {name=ReactivityOffMask, nodeColor=<rgba 0 1 0.6719 1>}
  n66 = ComponentMask(None=NamedRerouteUsage()) {b=true, r=true}
  n67 = TextureSample!(UVs=Multiply(A=Multiply(A=@n66) {constB=0.04}, B=Constant2Vector() {g=2.5, r=1}), Tex, Apply View MipBias) {samplerType=SAMPLERTYPE_LinearColor, texture=&T_MagmaBreak_Mask}
  n68 = OneMinus(None=NamedRerouteUsage())
  n69 = InverseLinearInterpolate(Value=@n63) {constA=0.6}
  n70 = LinearInterpolate(A=Saturate(None=InverseLinearInterpolate(Value=@n63) {constB=0.8}), B=Multiply(A=@n69) {constB=1.6}, Alpha=Saturate(None=Ceil(None=@n69))) {desc=MakeFXContinuous}
  n71 = SmoothStep(Min=@n67.R, Value=SmoothStep(Value=InverseLinearInterpolate(A=@n68, Value=@n70) {constA=0.5}) {constMax=0.25})
  NamedRerouteDeclaration(None=LinearInterpolate(B=Saturate(None=@n71), Alpha=ScalarParameter() {defaultValue=1, group=MagmaBreak, parameterName=ReactivityOn})) {name=ReactivityMask, nodeColor=<rgba 0.1719 0 1 1>}
  NamedRerouteDeclaration(None=FunctionInput() {inputName=Mask, sortPriority=1}) {name=InMask, nodeColor=<rgba 1 0 0.9141 1>}
  n72 = Time()
  n73 = ScalarParameter() {group=MagmaBreak, parameterName=Timestamp}
  n74 = Constant() {r=1.5}
  n75 = Clamp(None=Divide(A=Subtract(A=@n72, B=@n73), B=@n74) {desc=InLength})
  n76 = Subtract(A=@n72, B=Add(A=@n73, B=Add(A=@n74) {constB=3, desc=eXTRAdURATION}))
  n77 = OneMinus(None=Clamp(None=Divide(A=@n76) {constB=3, desc=OutLength}))
  NamedRerouteDeclaration(None=StaticSwitchParameter(True=Multiply(A=@n75, B=@n77), False=Constant() {r=0.0001}) {group=MagmaBreak, parameterName=UseBackpackReactivity}) {name=Reactivity, nodeColor=<rgba 1 0 0.9141 1>}
  n78 = NamedRerouteUsage()
  n79 = Multiply(A=LinearInterpolate(A=@n78, Alpha=@n3), B=Add(A=Constant() {r=1}, B=NamedRerouteUsage()))
  n80 = Add(
   A=Multiply(A=NamedRerouteUsage(), B=@n4)
   B=Multiply(A=NamedRerouteUsage(), B=Multiply(A=@n78, B=@n79)) {desc=Restore Original Emissive Outside of Cavity}
  )
  n81 = NamedRerouteUsage()
  n82 = ScalarParameter() {group=MagmaBreak, parameterName=PreviewPosition}
  NamedRerouteDeclaration(None=Multiply(
    A=LinearInterpolate(A=@n80, B=@n81, Alpha=@n82) {desc=Debug}
    B=ScalarParameter() {defaultValue=1.35, group=MagmaBreak, parameterName=OverallEmissiveIntensity}
   ) {constB=1.25}) {name=OutEmissive, nodeColor=<rgba 0.9688 0 1 1>}
  n83 = OneMinus(None=NamedRerouteUsage())
  n84 = Constant() {r=1}
  n85 = NamedRerouteUsage()
  n86 = Multiply(
   A=Power(Base=Multiply(A=@n11.R, B=@n83), Exp=Add(A=@n84, B=@n85))
   B=ScalarParameter() {defaultValue=2, group=MagmaBreak, parameterName=OutterRimIntensity}
  ) {constB=4}
  n87 = NamedRerouteUsage()
  NamedRerouteDeclaration(None=Multiply(A=Saturate(None=@n86), B=OneMinus(None=@n87))) {name=OutterRim, nodeColor=<rgba 0.6406 0 1 1>}
  n88 = TextureSampleParameter2D() {
   group=MagmaBreak
   parameterName=MagmaFXMask
   samplerType=SAMPLERTYPE_Masks
   texture=/BRCosmetics/Characters/Player/Male/Medium/Bodies/M_MED_MagmaBreak/Textures/T_MagmaBreak_Body_FX
  }
  NamedRerouteDeclaration(None=@n88.RGB) {name=MagmaFXMask, nodeColor=<rgba 0 1 0.7656 1>}
  NamedRerouteDeclaration(None=GetMaterialAttributes(None=NamedRerouteUsage()) {attributeGetTypes [#3, #4, #1]}.Normal) {name=InNormalMap, nodeColor=<rgba 0.8047 0 1 1>}
  OneMinus(None=SphereMask(A=NamedRerouteUsage(), B=Constant3Vector() {constant=<rgba 0 0 -23 0>}) {attenuationRadius=25})
 ]}
