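# Parent skeletal-mesh character material for MagmaBreak's CrewStyle skins, built between the 2023 CharacterParent PreFX and PostFX functions.
# MF_MagmaBreak feeds Ice, Galaxy, optional gold carbon-fiber and holographic-glitter layers, with a UseCrewStyle switch falling back to plain MagmaBreak.
# The saturated magma mask blackens styled base color and restores magma emissive over the style layers.
# Gold crew variants also blend carbon-fiber normals into the clear-coat bottom normal.
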Material /BRCosmetics/Characters/Player/Male/Medium/Bodies/M_MED_MagmaBreak/Skins/CrewStyle/M_MagmaBreak_Crew_Parent
graph {
 properties {
  bUseMaterialAttributes=true
  bUsedWithMeshDeformer=true
  bUsedWithMorphTargets=true
  bUsedWithSkeletalMesh=true
  pixelDepthOffsetMode=PDOM_Legacy
  shadingModel=MSM_FromMaterialExpression
  subsurfaceProfile=/Game/Packages/Fortress_SharedMaterials/SubSurfaceProfile/Character_SubsurfaceProfile
  usedShadingModels=Default Lit | Subsurface | Clear Coat | Subsurface Profile | Cloth
 }
 nodes [
  n1 = MaterialFunctionCall() {materialFunction=&MF_CharacterParent_2023_PreFX}
  n2 = MaterialFunctionCall!(In=@n1.Result, Mask=@n1.Result) {materialFunction=/BRCosmetics/Characters/Player/Male/Medium/Bodies/M_MED_MagmaBreak/Materials/MF_MagmaBreak}
  n3 = MaterialFunctionCall!(
   Tech Art Mask=@n1.Tech Art Mask
   Mask=@n1.Tech Art Mask
   In=MaterialFunctionCall!(TechArt=@n1.Tech Art Mask, Fresnel=@n1.Tech Art Mask, In=@n2.Result) {materialFunction=/Game/Global/Materials/Character_Parent_2023/CrewStyle_Parent_2023/Ice/MF_CrewStyle_Ice_Inst}.Result
  ) {materialFunction=/Game/Global/Materials/Character_Parent_2023/CrewStyle_Parent_2023/Galaxy/MF_CrewStyle_Galaxy_Inst}
  n4 = MaterialFunctionCall!(Alpha=@n1.Tech Art Mask, In=@n3.Result) {materialFunction=/Game/Global/MaterialFunctions/TechArt/MF_CarbonFiberDualLayer}
  n5 = GetMaterialAttributes(None=@n4.Result) {attributeGetTypes [#1]}
  n6 = LinearInterpolate(
   A=@n5.Normal
   B=MaterialFunctionCall(
    BaseNormal=@n5.Normal
    AdditionalNormal=TextureSample(UVs=@n4.CarbonFiberUvs) {samplerType=SAMPLERTYPE_Normal, texture=/Game/Weapons/WeaponSkins/Wraps/Textures/T_Wrap_CarbonFiber_N}.RGB
   ) {materialFunction=&BlendAngleCorrectedNormals}.Result
   Alpha=@n4.Mask
  )
  ClearCoatNormalCustomOutput!(None=StaticSwitchParameter(True=StaticSwitchParameter(True=@n6, False=@n5.Normal) {parameterName=isGold}, False=@n5.Normal) {group=MagmaBreak, parameterName=UseCrewStyle})
  n7 = GetMaterialAttributes(None=MaterialFunctionCall!(Alpha=@n1.Tech Art Mask, In=StaticSwitchParameter(True=@n4.Result, False=@n3.Result) {parameterName=isGold}, Fresnel=@n1.Tech Art Mask) {materialFunction=/Game/Global/MaterialFunctions/TechArt/MF_GlitterHolographic}.Result) {attributeGetTypes [#2, #3]}
  n8 = Saturate(None=Multiply(A=@n2.MagmaMask))
  n9 = LinearInterpolate(A=@n7.Base Color, Alpha=@n8) {constB=0}
  n10 = LinearInterpolate(
   A=@n7.Emissive Color
   B=GetMaterialAttributes(None=@n2.Result) {attributeGetTypes [#3]}.Emissive Color
   Alpha=@n8
  )
  MaterialFunctionCall(
   Material=StaticSwitchParameter(
    True=SetMaterialAttributes!(MaterialAttributes=@n7.MaterialAttributes, Base Color=@n9, Emissive Color=@n10) {desc=OVERRIDE CREWSTYLE EMISSIVE, attributeSetTypes [#2, #3]}
    False=@n2.Result
   ) {group=MagmaBreak, parameterName=UseCrewStyle}
   M=@n1.M
   Pre Skinned Local Position=@n1.M
   Distance Blend=@n1.M
   Fresnel=@n1.M
  ) {materialFunction=&MF_CharacterParent_2023_PostFX}
 ]
}
